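# Translucent unlit material for the animated electric waves FX on the Jumpstart character's skeletal mesh.
# Three panning lightning-mask samples get stepped sine-based time jitter and gradient-texture warping, combining into flickering background and spark layers.
# That mask drives a yellow curve-atlas color remapped to HDR, plus depth-faded opacity through the shared character opacity function.
# Vertex colors vary timing, bend and fade; regular and staccato gradient pulses push world position offset.
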Material /BRCosmetics/Characters/Player/Male/Medium/Bodies/M_MED_Jumpstart/FX/M_JumpStart_ElectricWaves
graph {
 properties {
  bUseMaterialAttributes=true
  bUsedWithMeshDeformer=true
  bUsedWithSkeletalMesh=true
  blendMode=BLEND_Translucent
  pixelDepthOffsetMode=PDOM_Legacy
  shadingModel=MSM_Unlit
  usedShadingModels=Unlit
 }
 nodes [
  n1 = VertexColor()
  n2 = Multiply(A=Add(A=Time(), B=Multiply(A=@n1.G) {constB=0.15}), B=ScalarParameter() {defaultValue=1, group=ElectricWaves, parameterName=Global_Speed})
  n3 = Multiply(A=@n2) {constB=0.5}
  n4 = Constant() {r=1.33}
  n5 = Multiply(A=Sine(None=@n3), B=Sine(None=Multiply(A=@n3, B=@n4)))
  n6 = Sine(None=Multiply(A=Add(A=@n2) {constB=0.33}) {constB=0.25})
  n7 = Round(None=@n6)
  n8 = Floor(None=Multiply(A=@n6) {constB=3})
  n9 = Abs(None=Add(A=Multiply(A=@n5, B=@n7), B=Divide(A=@n8) {constB=3}))
  Frac(None=@n9)
  n10 = Frac(None=Multiply(A=@n2) {constB=4})
  n11 = VertexColor()
  n12 = Add(
   A=Multiply(A=Time(), B=ScalarParameter() {defaultValue=0.5, group=ElectricWaves, parameterName=Sparks_Speed, sortPriority=1}) {constB=0.33}
   B=@n11.G
  )
  n13 = Frac(None=Multiply(A=Add(A=@n12) {constB=0.1}) {constB=0.9})
  n14 = Multiply(A=@n13, B=@n13)
  n15 = Saturate(None=Add(A=Multiply(A=@n10) {constB=0.25}, B=Add(A=@n14) {constB=0.5}) {constB=0.5})
  n16 = OneMinus(None=VertexColor().R)
  n17 = Multiply(A=@n13, B=@n13)
  n18 = Floor(None=Multiply(A=@n9) {constB=5})
  n19 = Add(A=@n2, B=Multiply(A=ScalarParameter() {defaultValue=-1.75, parameterName=Jitter Intensity}, B=Divide(A=@n18) {constB=5}) {constB=-1.5})
  n20 = VertexColor()
  n21 = ScalarParameter() {defaultValue=0.15, group=ElectricWaves, parameterName=Sparks_Bend, sortPriority=2}
  n22 = Add(A=TextureCoordinate() {uTiling=4}, B=AppendVector(A=Multiply(A=@n20.R, B=@n21), B=Constant()))
  n23 = Add(A=Add(A=@n19) {constB=0.33}, B=ComponentMask(None=@n22) {r=true})
  n24 = Add(A=ComponentMask(None=@n22) {g=true}) {constB=0.2}
  n25 = Add(A=ComponentMask(None=@n22) {g=true}) {constB=-0.08}
  n26 = TextureCoordinate() {uTiling=3}
  n27 = Multiply(A=VertexColor().R, B=ScalarParameter() {defaultValue=0.1, group=ElectricWaves, parameterName=Deform_Bend, sortPriority=10})
  n28 = VectorParameter() {group=ElectricWaves, parameterName=Background_DeformPannerSpeed, sortPriority=3, defaultValue=<rgba 0.25 0 0 0>}
  n29 = TextureSample(UVs=Panner(Coordinate=Add(A=@n26, B=@n27), Speed=ComponentMask(None=@n28.RGB) {g=true, r=true})) {
   samplerType=SAMPLERTYPE_LinearColor
   texture=&/T_Gradient_Lightning_JumpStart
  }
  n30 = Multiply(A=@n29.R)
  n31 = VectorParameter() {group=ElectricWaves, parameterName=Background_PannerSpeed, sortPriority=3, defaultValue=<rgba -0.1 0 0 0>}
  n32 = ComponentMask(None=@n31.RGB) {g=true, r=true}
  n33 = Panner(Coordinate=LinearInterpolate(A=AppendVector(A=@n23, B=@n24), B=AppendVector(A=@n23, B=@n25), Alpha=@n30), Speed=@n32)
  n34 = Constant2Vector() {g=0.95, r=0.5}
  n35 = TextureSample!(UVs=Subtract(A=Multiply(A=@n33, B=@n34), B=Constant2Vector()), Tex, Bias, Apply View MipBias) {
   mipValueMode=TMVM_MipBias
   samplerType=SAMPLERTYPE_LinearColor
   texture=/Game/Effects/Fort_Effects/Textures/Electric/T_Lightning_Mask_Background
  }
  n36 = Add(A=Add(A=@n19) {constB=0.2}, B=ComponentMask(None=@n22) {r=true})
  n37 = Add(A=ComponentMask(None=@n22) {g=true}) {constB=0.1}
  n38 = Add(A=ComponentMask(None=@n22) {g=true}) {constB=-0.05}
  n39 = Panner(Coordinate=LinearInterpolate(A=AppendVector(A=@n36, B=@n37), B=AppendVector(A=@n36, B=@n38), Alpha=@n30), Speed=@n32)
  n40 = Constant2Vector() {g=0.9, r=0.5}
  n41 = TextureSample(UVs=Subtract(A=Multiply(A=@n39, B=@n40), B=Constant2Vector() {g=-0.05})) {
   mipValueMode=TMVM_MipBias
   samplerType=SAMPLERTYPE_LinearColor
   texture=&T_Lightning_Mask_Background
  }
  n42 = ScalarParameter() {defaultValue=0.45, group=ElectricWaves, parameterName=Background_Intensity}
  n43 = Multiply(A=Multiply(A=Multiply(A=@n35.A) {constB=0.75}, B=OneMinus(None=@n41.A)), B=@n42) {constB=0.1}
  n44 = OneMinus(None=@n41.R)
  n45 = Constant() {r=1}
  n46 = Saturate(None=Add(A=MaterialFunctionCall(Alpha=@n44, Min=@n17, Max=@n45) {materialFunction=&SmoothStep}.Result) {constB=0.2})
  n47 = OneMinus(None=@n41.R)
  n48 = Multiply(A=Multiply(A=Power(Base=@n47) {constExponent=3}, B=Multiply(A=@n41.A) {constB=1.5}), B=@n42) {constB=0.1}
  n49 = Add(A=Multiply(A=Multiply(A=@n16, B=@n16)) {constB=0.425}, B=Add(A=Multiply(A=@n17, B=@n43), B=Multiply(A=@n46, B=@n48)) {constB=0})
  n50 = Multiply(A=@n12, B=Add(A=Multiply(A=@n11.G) {constB=0.15}))
  n51 = Add(A=@n19, B=ComponentMask(None=@n22) {r=true})
  n52 = Add(A=ComponentMask(None=@n22) {g=true}) {constB=0.1}
  n53 = Add(A=ComponentMask(None=@n22) {g=true}) {constB=-0.1}
  n54 = VectorParameter() {group=ElectricWaves, parameterName=Sparks_PannerSpeed, sortPriority=3, defaultValue=<rgba -0.25 0 0 0>}
  n55 = Panner(
   Coordinate=LinearInterpolate(A=AppendVector(A=@n51, B=@n52), B=AppendVector(A=@n51, B=@n53), Alpha=@n30)
   Speed=ComponentMask(None=@n54.RGB) {g=true, r=true}
  )
  n56 = Constant2Vector() {g=1.05, r=0.5}
  n57 = TextureSample(UVs=Subtract(A=Multiply(A=@n55, B=@n56), B=Constant2Vector() {g=-0.05})) {
   mipValueMode=TMVM_MipBias
   samplerType=SAMPLERTYPE_LinearColor
   texture=&T_Lightning_Mask_Background
  }
  n58 = Subtract(A=Add(A=Frac(None=@n50), B=@n57.R))
  n59 = Saturate(None=OneMinus(None=Abs(None=@n58)))
  n60 = MaterialFunctionCall(Alpha=Multiply(A=@n59, B=@n59), Min=Constant() {r=0.925}, Max=Constant() {r=1}) {materialFunction=&SmoothStep}
  n61 = Subtract(A=Add(A=@n13, B=@n41.R))
  n62 = Saturate(None=OneMinus(None=Abs(None=@n61)))
  n63 = Multiply(A=@n62, B=@n62)
  n64 = Constant() {r=0.925}
  n65 = Constant() {r=1}
  n66 = Add(
   A=Multiply(A=Multiply(A=@n60.Result, B=@n57.A), B=OneMinus(None=@n41.A))
   B=Multiply(
    A=MaterialFunctionCall(Alpha=@n63, Min=@n64, Max=@n65) {materialFunction=&SmoothStep}.Result
    B=@n41.A
   )
  ) {constB=0}
  n67 = Add(A=Multiply(A=@n10, B=Multiply(A=@n66) {constB=10}), B=Saturate(None=Multiply(A=@n15, B=@n49)))
  n68 = CurveAtlasRowParameter(CurveTime=Power(Base=@n67, Exp=ScalarParameter() {defaultValue=1, group=Color, parameterName=Curve Contrast Controller, sortPriority=1})) {
   atlas=&AthenaCurveAtlas01
   curve=/Game/Characters/Player/Male/Medium/Bodies/M_MED_Jumpstart/FX/JumpStart_Yellow
   defaultValue=120
   group=Color
   parameterName=Curve Parameter
  }
  n69 = Multiply(A=@n68, B=VectorParameter() {group=Color, parameterName=HDR_Tint, sortPriority=2, defaultValue=<rgba 1.85 2.3 0.05 0>}.RGB)
  n70 = Clamp(None=@n69)
  n71 = ScalarParameter() {defaultValue=1, group=Color, parameterName=HDR_Intensity, sortPriority=3}
  n72 = ScalarParameter() {defaultValue=1, group=Color, parameterName=Derive HDR Power, sortPriority=3}
  n73 = ScalarParameter() {defaultValue=1, group=Alpha, parameterName=MasterAlpha}
  n74 = Multiply(
   A=DepthFade(Opacity=@n67, FadeDistance=ScalarParameter() {defaultValue=4, group=Alpha, parameterName=Fade Distance, sortPriority=1})
   B=OneMinus(None=VertexColor().B)
  )
  n75 = Time()
  n76 = VertexColor()
  n77 = Sine(None=Add(A=Multiply(A=@n75) {constB=0.25}, B=Multiply(A=@n76.G)))
  n78 = Multiply(A=@n77) {constB=0.75}
  n79 = VectorParameter() {group=WPO_ElectricWaves, parameterName=Regular Pulse Vector A, sortPriority=2, defaultValue=<rgba 0 -4 3.5 1>}
  n80 = VectorParameter() {group=WPO_ElectricWaves, parameterName=Regular Pulse Vector, sortPriority=2, defaultValue=<rgba -2 -2 3 1>}
  n81 = ComponentMask(None=TextureCoordinate() {vTiling=0.5}) {g=true}
  n82 = ScalarParameter() {defaultValue=2, group=WPO_ElectricWaves, parameterName=WPO_Push Contrast}
  n83 = Multiply(A=Abs(None=Add(A=@n78) {constB=0.25}), B=LinearInterpolate(A=@n79.RGB, B=@n80.RGB, Alpha=Power(Base=@n81, Exp=@n82)))
  n84 = OneMinus(None=ComponentMask(None=TextureCoordinate()) {g=true})
  n85 = Clamp(None=ComponentMask(None=MaterialFunctionCall() {materialFunction=&LocalPosition}."Local Position (Excluding Offsets)") {r=true}) {minDefault=-1}
  n86 = TextureCoordinate() {uTiling=2}
  n87 = ComponentMask(None=@n86) {r=true}
  n88 = AppendVector(A=Add(A=@n2, B=Multiply(A=@n85, B=@n87)), B=ComponentMask(None=@n86) {g=true})
  n89 = VectorParameter() {group=WPO_ElectricWaves, parameterName=Regular Pulse Panner Speed, sortPriority=4, defaultValue=<rgba 0 0.25 0 0>}
  n90 = TextureSample!(
   UVs=Panner(
    Coordinate=MaterialFunctionCall(XY=@n88) {materialFunction=&Swizzle}.YX
    Speed=ComponentMask(None=@n89.RGB) {g=true, r=true}
   )
   Tex
   Apply View MipBias
  ) {samplerSource=SSM_Wrap_WorldGroupSettings, texture=/Game/Global/Textures/Default/Masks_and_Gradients/T_Vert_Gradient_Soft}
  n91 = Subtract(A=Multiply(A=@n90.R) {constB=2})
  n92 = Multiply(
   A=Multiply(A=Transform(None=@n83) {transformSourceType=TRANSFORMSOURCE_Local}, B=Multiply(A=@n84, B=@n91))
   B=ScalarParameter() {defaultValue=1.25, group=WPO_ElectricWaves, parameterName=Regular Pulse WPO Amount}
  )
  n93 = VectorParameter() {group=WPO_ElectricWaves, parameterName=Staccato Pulse Vector A, sortPriority=2, defaultValue=<rgba 1 -2 0.5 1>}
  n94 = VectorParameter() {group=WPO_ElectricWaves, parameterName=Staccato Pulse Vector, sortPriority=3, defaultValue=<rgba 0.5 -2 0.5 1>}
  n95 = ComponentMask(None=TextureCoordinate() {uTiling=0.25}) {g=true}
  n96 = ScalarParameter() {defaultValue=2, group=WPO_ElectricWaves, parameterName=WPO_Push Contrast}
  n97 = Multiply(A=LinearInterpolate(A=@n93.RGB, B=@n94.RGB, Alpha=Power(Base=@n95, Exp=@n96)), B=Abs(None=OneMinus(None=@n77)))
  n98 = OneMinus(None=ComponentMask(None=TextureCoordinate()) {g=true})
  n99 = TextureCoordinate()
  n100 = AppendVector(A=Multiply(A=@n85, B=ComponentMask(None=@n99) {r=true}), B=ComponentMask(None=@n99) {g=true})
  n101 = VectorParameter() {group=WPO_ElectricWaves, parameterName=Staccato Pulse Panner Speed, sortPriority=5, defaultValue=<rgba 0 0.25 0 0>}
  n102 = TextureSample(UVs=Panner(
    Coordinate=MaterialFunctionCall(XY=@n100) {materialFunction=&Swizzle}.YX
    Speed=ComponentMask(None=@n101.RGB) {g=true, r=true}
   )) {samplerSource=SSM_Wrap_WorldGroupSettings, texture=/Game/Global/Textures/Default/Masks_and_Gradients/T_Vert_Gradient_Sections}
  n103 = Multiply(A=@n102.R, B=@n102.R)
  n104 = Multiply(
   A=Multiply(A=Transform(None=@n97) {transformSourceType=TRANSFORMSOURCE_Local}, B=Multiply(A=@n98, B=@n103))
   B=ScalarParameter() {defaultValue=1.25, group=WPO_ElectricWaves, parameterName=Staccato Pulse WPO Amount, sortPriority=1}
  )
  MaterialFunctionCall!(Material=MakeMaterialAttributes!(
    BaseColor
    Metallic
    Specular
    Roughness
    Anisotropy
    EmissiveColor=MaterialFunctionCall(LDR Input=@n70, HDR Tint=@n69, HDR Intensity Multiplier=@n71, Derive HDR Power=@n72) {materialFunction=&DeriveHDRfromLDR}.RemappedRange
    Opacity=Multiply(A=@n73, B=@n74)
    OpacityMask
    Normal
    Tangent
    WorldPositionOffset=Add(A=@n92, B=@n104) {constB=0}
    SubsurfaceColor
    ClearCoat
    ClearCoatRoughness
    AmbientOcclusion
    Refraction
    CustomizedUVs_0
    CustomizedUVs_1
    CustomizedUVs_2
    CustomizedUVs_3
    CustomizedUVs_4
    CustomizedUVs_5
    CustomizedUVs_6
    CustomizedUVs_7
    PixelDepthOffset
    ShadingModel
    Displacement
   )) {materialFunction=/Game/Packages/Fortress_SharedMaterials/Base_Material_Functions/BaseCharacterShaderFunctions/BaseFN_CharacterOpacity}
  Multiply(A=@n90.R, B=@n90.R)
 ]
}
