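# Niagara system for short yellowish electrical discharge flashes on the hands of the Jumpstart character cosmetic.
# Minimal emitter spawns at randomized loop intervals, pins particles to partition-indexed skeletal mesh bones or sockets, and generates location events.
# Flash, which has no spawn module of its own, draws brief mesh, sprite and light bursts with curve-driven material parameters and flickering alpha.
# User.bisHandHidden sets visibility tags on both emitters and suppresses location events.
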NiagaraSystem /BRCosmetics/Characters/Player/Male/Medium/Bodies/M_MED_Jumpstart/FX/NS_Discharges_JumpStart_Hands
dependencies {
 usedDataInterfaces [&NiagaraDataInterfaceCurve, &NiagaraDataInterfaceSkeletalMesh]
 usedDynamicInputs []
 usedModules [
  &SystemState
  &self:SystemSpawnScript.SystemScriptSource.SystemScriptGraph.NiagaraNodeAssignment_0.SetVariables_#1
 ]
}
emitters {
 Flash {
  emitterUpdateScript [&EmitterState()]
  particleSpawnScript [
   &InitializeParticle(
    Color=<rgba 1.75 1.8 1.25 1>
    Lifetime Max=0.15
    Lifetime Min=0.05
    Lifetime Mode=::NewEnumerator1
    Mesh Scale Mode=::NewEnumerator1
    Mesh Uniform Scale Min=1.5
    Sprite Size Mode=::NewEnumerator3
    Uniform Sprite Size=30
    UsePositionOffset=false
    Write Material Random=true
   )
   &AddVelocity(Velocity=&RandomRangeVector(Minimum=<-4 -0.5 1.5>, Maximum=<4 -1 3>))
   &InitialMeshOrientation(Use Rotation Vector=false)
  ]
  particleUpdateScript [
   &ParticleState()
   &ScaleMeshSize(Scale Factor=&RandomRangeVector(Minimum=<0.5 0.55 0.6>, Maximum=<0.5 0.6 0.6>))
   &SolveForcesAndVelocity(Clamp Velocity=false, Limit Acceleration=false)
   &DynamicMaterialParameters(
    Index 0 Param 1=&FloatFromCurve(FloatCurve=DataInterface() {propertyValues {Curve {keys [
         {arriveTangent=322.0399, interpMode=RCIM_Cubic, leaveTangent=322.0399, tangentMode=RCTM_User, time=0, value=0}
         {arriveTangent=1, interpMode=RCIM_Cubic, leaveTangent=1, time=0.02, value=2.4407}
         {arriveTangent=-1.8841, interpMode=RCIM_Cubic, leaveTangent=-1.8841, tangentMode=RCTM_User, time=1, value=1}
        ]}}})
    Index 0 Param 2=&FloatFromCurve(FloatCurve=DataInterface() {propertyValues {Curve {keys [
         <at 0 0 RCIM_Cubic>
         {arriveTangent=46.5177, interpMode=RCIM_Cubic, leaveTangent=46.5177, time=0.04, value=5.4468}
         {arriveTangent=-5.6738, interpMode=RCIM_Cubic, leaveTangent=-5.6738, time=0.0537, value=2.5}
         <at 1 0 RCIM_Cubic>
        ]}}})
    Index 0 Param 3=&RandomRangeFloat(Minimum=0, Maximum=1)
    Index 0 Param 4=&RandomRangeFloat(Minimum=0, Maximum=1)
    Module.Index0Float0Write=true
    Module.Index0Float1Write=true
    Module.Index0Float2Write=true
    Module.Index0Float3Write=true
   )
   &ScaleColor(
    Scale Alpha=&FloatFromCurve(FloatCurve=DataInterface() {propertyValues {Curve {keys [
         <at 0.4 1 RCIM_Cubic>
         {arriveTangent=-7.4538, interpMode=RCIM_Cubic, leaveTangent=-7.4538, time=0.5, value=1.12}
         {arriveTangent=0.3409, interpMode=RCIM_Cubic, leaveTangent=0.3409, time=0.53, value=0.031}
         {arriveTangent=-0.0097, interpMode=RCIM_Cubic, leaveTangent=-0.0097, time=0.588, value=1.15}
         {arriveTangent=0.1389, interpMode=RCIM_Cubic, leaveTangent=0.1389, time=0.633, value=0.03}
         {arriveTangent=-0.1493, interpMode=RCIM_Cubic, leaveTangent=-0.1493, time=0.66, value=1.16}
         {arriveTangent=-4.4, interpMode=RCIM_Cubic, leaveTangent=-4.4, time=0.7, value=0.02}
         {arriveTangent=-0.0909, interpMode=RCIM_Cubic, leaveTangent=-0.0909, time=0.76, value=0.72}
         {arriveTangent=-0.3333, interpMode=RCIM_Cubic, leaveTangent=-0.3333, time=0.81, value=0.01}
         <at 0.85 0.69 RCIM_Cubic>
         {arriveTangent=-2.65, interpMode=RCIM_Cubic, leaveTangent=-2.65, time=0.9, value=0.01}
         {arriveTangent=-0.1, interpMode=RCIM_Cubic, leaveTangent=-0.1, time=0.95, value=0.425}
         <at 1 0 RCIM_Cubic>
        ]}}})
    ScaleA=true
    ScaleRGB=false
   )
   /SetVariables(Particles.VisibilityTag=&MakeCustomIntFromBool(Bool=$User.bisHandHidden, True INT=1, False INT=0)) {isSetParameters=true, module=SetVariables_#2}
   /SetVariables(Particles.SpriteColor=&MultiplyLinearColor(
     Linear Color=<rgba 1.5 1.75 0 1>
     Scale Factor=&MakeVector4FromVectorAndFloat(
      "Vector (RGB)"=&VectorFromFloat(Value=&FloatFromCurve(FloatCurve=DataInterface() {propertyValues {Curve {keys [
            {arriveTangent=48.9308, interpMode=RCIM_Cubic, leaveTangent=48.9308, tangentMode=RCTM_User, time=0, value=1}
            <at 0.03 2 RCIM_Cubic>
            <at 1 1 RCIM_Cubic>
           ]}}}))
      "Float (Alpha)"=&FloatFromCurve(
       FloatCurve=DataInterface() {propertyValues {Curve {keys [
           <at 0 1 RCIM_Cubic>
           {arriveTangent=-0.7849, interpMode=RCIM_Cubic, leaveTangent=-0.7849, tangentMode=RCTM_User, time=0.8, value=0.8}
           <at 1 0 RCIM_Cubic>
          ]}}}
       Scale Curve=0.01
      )
     )
    )) {isSetParameters=true}
   &Light_Attributes(Color=1, Color Mode=0, Falloff Exponent=1, Light Color:LinearColor=<rgba 1 1.1 0.5 8>, Light Exponent=16, Light Radius=40, Radius=1)
  ]
  rendererClasses [
   &NiagaraMeshRendererProperties
   &NiagaraSpriteRendererProperties
   &NiagaraLightRendererProperties
  ]
  renderers [
   {rendererClass=&NiagaraMeshRendererProperties, rendererIndex=0}
   {rendererClass=&NiagaraSpriteRendererProperties, rendererIndex=1}
   {rendererClass=&NiagaraLightRendererProperties, rendererIndex=2}
  ]
 }
 Minimal {
  emitterUpdateScript [
   &EmitterState(
    Life Cycle Mode=::NewEnumerator1
    Loop Duration=&RandomRangeFloat(Minimum=1.3, Maximum=2.25)
    Recalculate Duration Each Loop=true
    UseLoopDelay=false
   )
   &SpawnRate(SpawnRate=3, Use Spawn Probability=false)
   &SpawnBurst_Instantaneous(Use Spawn Probability=false, UseLoopCountLimit=false)
  ]
  particleSpawnScript [
   &PartitionParticles(Partition Mode:ENiagara_PartitionMode_v2=::NewEnumerator5)
   &InitializeParticle(Color Mode=::NewEnumerator0, Lifetime=0.25, UsePositionOffset=false, Write Material Random=true)
   &ApplyInitialForces()
  ]
  particleUpdateScript [
   &SolveForcesAndVelocity(Clamp Velocity=false, Limit Acceleration=false)
   &ParticleState()
   &SkeletalMeshLocation(
    Bone / Socket Index=$Particles.PartitionParticles.Partition
    Bone Sampling Mode=::NewEnumerator4
    MaskSpawnGroup=false
    Sampled Position Offset=<9 -2 0>
    Skeletal Mesh=$System.Skeletal Mesh
   )
   /SetVariables(Particles.VisibilityTag=&MakeCustomIntFromBool(Bool=$User.bisHandHidden, True INT=1, False INT=0)) {isSetParameters=true, module=SetVariables_#2}
   &GenerateLocationEvent(
    Event Generation Enabled:NiagaraBool=&Boolean_NotOperation(Bool=$User.bisHandHidden)
    Use Event Probability=false
    UseEventDelay=false
   )
  ]
  rendererClasses [&NiagaraSpriteRendererProperties]
  renderers [{rendererClass=&NiagaraSpriteRendererProperties}]
 }
}
summary {
 systemName=NS_Discharges_JumpStart_Hands
 emitters [
  {emitterName=Minimal, rendererClasses [&NiagaraSpriteRendererProperties]}
  {
   emitterName=Flash
   rendererClasses [
    &NiagaraMeshRendererProperties
    &NiagaraSpriteRendererProperties
    &NiagaraLightRendererProperties
   ]
  }
 ]
 userVariables [User.bisHandHidden {type=NiagaraBool}]
}
