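# Physics asset for the Jumpstart medium male player body, covering its dangling accessories: battery, jumper cables, and a holstered baton with ribbon.
# Capsules on the upper spine and both thigh twist bones provide body shapes alongside the accessory bodies.
# The battery hangs from spine_04, both cables from a master bone, and the baton chains master to holster to baton to ribbon.
# Accessory constraints use angular limits, tightest on the baton within its holster, while the spine joint rotates freely.
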PhysicsAsset /BRCosmetics/Characters/Player/Male/Medium/Bodies/M_MED_Jumpstart/Meshes/M_MED_Jumpstart_PhysicsAsset
bodies {
 Dyn_baton [{shapeType=Capsule, length=40.8448, radius=1.6156, center=<4.3423 0 0>, rotation=<rot 90 0 0>}]
 Dyn_baton_MSTR [{shapeType=Box, extentX=1, extentY=1, extentZ=1}]
 Dyn_baton_holster [{shapeType=Box, extentX=6.25, extentY=11, extentZ=25, center=<8.7142 0.507 -1.3826>, rotation=<rot 90 0 0>}]
 Dyn_baton_ribbon [{shapeType=Box, extentX=4.6038, extentY=2.11, extentZ=6.5673, center=<9.0413 0.2576 -0.3829>, rotation=<rot 0 -87.0244 -86.1285>}]
 Dyn_battery [{shapeType=Box, extentX=21.0958, extentY=13.9246, extentZ=28.1142, center=<-8.2918 1.2591 0>}]
 Dyn_jumper_cable_l [{shapeType=Box, extentX=5, extentY=5, extentZ=13.2684, center=<-5 0 0>, rotation=<rot 90 0 0>}]
 Dyn_jumper_cable_mstr [{shapeType=Box, extentX=7.5, extentY=7.5, extentZ=7.5}]
 Dyn_jumper_cable_r [{shapeType=Box, extentX=5, extentY=5, extentZ=13.2684, center=<-5 0 0>, rotation=<rot 90 0 0>}]
 spine_03 [{shapeType=Capsule, length=17.6239, radius=10.6979}]
 spine_04 [{shapeType=Capsule, length=16.8567, radius=10.4182}]
 thigh_twist_01_l [{shapeType=Capsule, length=21.5023, radius=8.4977, center=<-3.6542 -0.0001 0>, rotation=<rot 90 0 0>}]
 thigh_twist_01_r [{shapeType=Capsule, length=23.125, radius=6.875, rotation=<rot 90 0 0>}]
}
constraints [
 {
  bone1Name=spine_04
  bone2Name=spine_03
  swing1LimitDegrees=45
  swing1Motion=Free
  swing2LimitDegrees=45
  swing2Motion=Free
  twistLimitDegrees=45
  twistMotion=Free
 }
 {bone1Name=Dyn_battery, bone2Name=spine_04, swing1LimitDegrees=45, swing2LimitDegrees=9, twistLimitDegrees=9}
 {bone1Name=Dyn_jumper_cable_l, bone2Name=Dyn_jumper_cable_mstr, swing1LimitDegrees=10, swing2LimitDegrees=45, twistLimitDegrees=45}
 {bone1Name=Dyn_baton_holster, bone2Name=Dyn_baton_MSTR, swing1LimitDegrees=10, swing2LimitDegrees=10, twistLimitDegrees=10}
 {bone1Name=Dyn_baton, bone2Name=Dyn_baton_holster, swing1LimitDegrees=1.5, swing2LimitDegrees=2.5, twistLimitDegrees=2.5}
 {bone1Name=Dyn_baton_ribbon, bone2Name=Dyn_baton, swing1LimitDegrees=35, swing2LimitDegrees=35, twistLimitDegrees=25}
 {bone1Name=Dyn_jumper_cable_r, bone2Name=Dyn_jumper_cable_mstr, swing1LimitDegrees=10, swing2LimitDegrees=45, twistLimitDegrees=45}
]
