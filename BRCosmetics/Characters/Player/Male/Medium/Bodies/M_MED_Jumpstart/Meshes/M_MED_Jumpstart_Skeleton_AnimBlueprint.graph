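# Animation Blueprint for the Jumpstart male body cosmetic, adding physics-driven secondary motion on top of the pose copied from the attached parent mesh.
# An additive yaw offsets the jumper cable master bone, with a second yaw switched on by crouch events through Crouch_Alpha, without blending.
# A RigidBody node then simulates the override physics asset in spine_01 space, using custom gravity, scaled-down component motion and the skydiving vector force as external force.
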AnimBlueprint /BRCosmetics/Characters/Player/Male/Medium/Bodies/M_MED_Jumpstart/Meshes/M_MED_Jumpstart_Skeleton_AnimBlueprint
target_skeleton=&/M_MED_Jumpstart_Skeleton
graphs [
 AnimGraph {nodes [
   n1 = "Transform (Modify) Bone - Bone: Dyn_jumper_cable_mstr"!(ComponentPose=Local To Component(LocalPose=Copy Pose From Mesh() {bUseAttachedParent=true}), Alpha=1, Rotation="0,20,0") -> (Pose) {
    rotationMode=BMM_Additive
    rotationSpace=BCS_BoneSpace
    alphaBoolBlend {blendOption=Linear}
    boneToModify {boneName=Dyn_jumper_cable_mstr}
    rotation {yaw=20}
   }
   n2 = "Transform (Modify) Bone - Bone: Dyn_jumper_cable_mstr"(ComponentPose=@n1, Alpha=Get Crouch_Alpha() -> (Crouch_Alpha:double), Rotation="0,15,0") {
    rotationMode=BMM_Additive
    rotationSpace=BCS_BoneSpace
    alphaBoolBlend {blendOption=Linear}
    boneToModify {boneName=Dyn_jumper_cable_mstr}
    rotation {yaw=15}
   }
   n3 = RigidBody(
    ComponentPose=@n2
    ExternalForce=Break Fort Anim Input Skydiving External Force(FortAnimInput_SkydivingExternalForce=Get SkydivingExternalForceInput()) -> (
     bUseSkydivingVectorForce=~
     SkydivingVectorForce=~
     HeadToPelvisDirection=~
     FloatingMultiplier=~
     FloatingAdditive=~
     DivingMultiplier=~
     DivingAdditive=~
     ParachutingMultiplier=~
     ParachutingAdditive=~
    ).SkydivingVectorForce
   ) {
    bClampLinearTranslationLimitToRefPose=true
    bOverrideWorldGravity=true
    overridePhysicsAsset=&/M_MED_Jumpstart_PhysicsAsset
    simulationSpace=BaseBoneSpace
    alphaBoolBlend {blendOption=Linear}
    baseBoneRef {boneName=spine_01}
    componentAppliedLinearAccClamp=<2000 2000 2000>
    componentLinearAccScale=<0.1 0.1 0.1>
    componentLinearVelScale=<0.1 0.1 0.1>
    overrideWorldGravity {z=-980}
   }
   Output Pose(Result=Component To Local(ComponentPose=@n3)) {layerGroup=None}
  ]}
 EventGraph {nodes [
   Event OnBeginCrouching()
   > Set Crouch_Alpha(Crouch_Alpha:double=1) -> (Output_Get:double)
   Event OnEndCrouching()
   > Set Crouch_Alpha(Crouch_Alpha=0)
   TryGetPawnOwner()
   Event BlueprintUpdateAnimation()
  ]}
]
variables [Crouch_Alpha {type=number}]
